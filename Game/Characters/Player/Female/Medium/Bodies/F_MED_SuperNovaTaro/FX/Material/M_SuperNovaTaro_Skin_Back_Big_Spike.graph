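# Unlit, translucent, two-sided Niagara mesh-particle material for the big glowing back spike effect on the SuperNovaTaro skin.
# Panning plasma coils, distorted by scrolling Perlin noise, blend tintable colors, with a vertical UV gradient adding a brighter, partly desaturated end.
# Opacity combines an inverted sphere mask, sine and UV edge falloffs, view-angle fade and depth fade to soften the mesh.
# Particle color tints emissive and particle alpha scales opacity.
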Material /Game/Characters/Player/Female/Medium/Bodies/F_MED_SuperNovaTaro/FX/Material/M_SuperNovaTaro_Skin_Back_Big_Spike
graph {
 properties {
  bUsedWithNiagaraMeshParticles=true
  blendMode=BLEND_Translucent
  pixelDepthOffsetMode=PDOM_Legacy
  shadingModel=MSM_Unlit
  twoSided=true
  usedShadingModels=Unlit
 }
 nodes [
  n1 = TextureCoordinate() {vTiling=1.2}
  n2 = TextureCoordinate() {vTiling=0.5}
  n3 = Multiply(A=TextureSample(UVs=Panner(Coordinate=@n2) {speedY=-0.3}) {samplerType=SAMPLERTYPE_LinearGrayscale, texture=&T_SimplePerlin_Greyscale}.R) {constB=0.2}
  n4 = TextureSample(UVs=Panner(Coordinate=Add(A=@n1, B=@n3)) {speedX=-0.5}) {samplerType=SAMPLERTYPE_LinearGrayscale, texture=/Game/Effects/Fort_Effects/Emotes/UltraLight/Texture/T_PlasmaCoils}
  n5 = LinearInterpolate(
   A=LinearInterpolate(
    A=VectorParameter() {parameterName=Color1, defaultValue=<rgba 0.0231 0 0.5 1>}.RGB
    B=VectorParameter() {parameterName=Color2, defaultValue=<rgba 2 0 0.0158 1>}.RGB
    Alpha=Multiply(A=@n4.R)
   )
   B=VectorParameter() {parameterName=Color3, defaultValue=<rgba 3 0 0.3642 1>}.RGB
  ) {constAlpha=0.05}
  n6 = Desaturation(
   None=Multiply(A=@n5, B=ScalarParameter() {defaultValue=5, parameterName=Top Brightness}) {constB=7}
   Fraction=ScalarParameter() {defaultValue=0.3, parameterName=Des}
  )
  n7 = ComponentMask(None=TextureCoordinate()) {g=true}
  n8 = Saturate(None=Multiply(A=Power(Base=@n7) {constExponent=0.7}) {constB=1.5})
  n9 = ParticleColor()
  MaterialFunctionCall(Emissive=Multiply(A=LinearInterpolate(A=@n6, B=@n5, Alpha=@n8), B=@n9.RGB)) {materialFunction=&MF_FX_EmissiveAdjustments}
  n10 = SphereMask(A=TextureCoordinate(), B=Constant2Vector() {g=1, r=0.55}, Radius=Constant() {r=0.5}) {hardnessPercent=0}
  n11 = ComponentMask(None=TextureCoordinate()) {r=true}
  n12 = Power(Base=Sine(None=@n11) {period=2})
  n13 = TextureCoordinate()
  n14 = Multiply(A=OneMinus(None=ComponentMask(None=@n13) {g=true})) {constB=5}
  n15 = Multiply(A=Power(Base=OneMinus(None=@n10)) {constExponent=10}, B=LinearInterpolate(A=Multiply(A=@n12), Alpha=Saturate(None=@n14)))
  n16 = Saturate(None=Multiply(A=Subtract(A=@n11) {constB=0.2}) {constB=5})
  n17 = Power(Base=@n11) {constExponent=1}
  n18 = Multiply(A=Subtract(A=OneMinus(None=@n17)) {constB=0.05}) {constB=10}
  n19 = Multiply(
   A=Multiply(A=Multiply(A=@n15, B=@n16), B=Saturate(None=@n18))
   B=MaterialFunctionCall("FadeSensitivity (optional)"=Constant() {r=8}) {materialFunction=&FadeBasedOnViewAngle}.Result
  )
  Multiply(A=@n9.A, B=Power(Base=DepthFade(Opacity=@n19) {fadeDistanceDefault=20}) {constExponent=0.5})
 ]
}
