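# Masked unlit black-hole effect material for the SuperNovaTaro player skin, flagged for Niagara mesh particles and ribbons.
# A Use Trail Texture switch picks panning screen-aligned nebula blended through scrolling smoke, or an optional static nebula texture.
# Particle color tints this, an optional Fresnel center point adds glow, and MF_FX_EmissiveAdjustments processes the sum.
# The mask multiplies particle alpha by trail edge fades or view-angle fade, plus an optional bottom cut.
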Material /Game/Characters/Player/Female/Medium/Bodies/F_MED_SuperNovaTaro/FX/Material/M_SuperNovaTaro_Skin_BlackHole
graph {
 properties {
  bUsedWithNiagaraMeshParticles=true
  bUsedWithNiagaraRibbons=true
  blendMode=BLEND_Masked
  pixelDepthOffsetMode=PDOM_Legacy
  shadingModel=MSM_Unlit
  usedShadingModels=Unlit
 }
 nodes [
  MaterialFunctionCall!(TextureResolution=Constant() {r=100}) {materialFunction=/Engine/Functions/Engine_MaterialFunctions02/Texturing/ScreenAlignedPixelToPixelUVs}
  MaterialFunctionCall(TextureResolution=ScalarParameter() {defaultValue=256, parameterName=NebulaTextureResolution}) {materialFunction=&ScreenAlignedPixelToPixelUVs}
  n1 = Subtract(A=OneMinus(None=Fresnel() {baseReflectFraction=0, exponent=1})) {constB=0.975}
  n2 = Multiply(
   A=VectorParameter() {parameterName=Center point color, defaultValue=<rgba 1.5 1.1328 1.1903 1>}.RGB
   B=Saturate(None=Multiply(A=@n1) {constB=10})
  )
  n3 = ScalarParameter() {parameterName=Use center point}
  n4 = TextureSampleParameter2D(UVs=Panner(Coordinate=TextureCoordinate() {uTiling=0.7, vTiling=0.8}) {speedX=0.2}) {parameterName=dissolution Texture, texture=&T_Inky_Smoke_Tile}
  n5 = LinearInterpolate(A=Constant3Vector() {constant=<rgba 0.0141 0.0027 0.03 1>}, B=Constant3Vector() {constant=<rgba 0.2135 0.0266 0.1457 1>}, Alpha=@n4.R)
  n6 = Multiply(
   A=MaterialFunctionCall!(Center Point World Position) {materialFunction=/Engine/Functions/Engine_MaterialFunctions02/UVs/WorldSpaceAlignedScreenCoordinates}.Offset and Scaled UVS
   B=Constant2Vector() {g=40, r=40}
  )
  n7 = ComponentMask(None=VectorParameter() {parameterName="Nebula&&StarfieldPannerSpeed", defaultValue=<rgba 0.1 0 0 0>}.RGB) {g=true, r=true}
  n8 = Multiply(A=TextureSampleParameter2D(UVs=Panner(Coordinate=@n6, Speed=@n7) {speedX=0.1}) {parameterName=Trail Texture, texture=/Game/Characters/Player/Male/Medium/Bodies/M_Med_Soldier_04/Skins/Galaxy/Textures/T_FN_Nebula}.RGB) {constB=3}
  n9 = Multiply(A=Add(B=LinearInterpolate(A=@n5, B=@n8, Alpha=@n4.R)))
  n10 = SphereMask(A=ComponentMask(None=TextureCoordinate()) {r=true}, B=Constant() {r=0.5}) {attenuationRadius=0.55, hardnessPercent=0}
  n11 = LinearInterpolate(
   A=@n9
   B=LinearInterpolate(A=Constant3Vector(), B=@n9, Alpha=Saturate(None=@n10))
   Alpha=ScalarParameter() {defaultValue=1, parameterName=Use Edge Power}
  )
  n12 = VectorParameter() {parameterName=Main color, defaultValue=<rgba 0.8 1 1.5 1>}
  n13 = Multiply(
   A=MaterialFunctionCall() {materialFunction=&WorldSpaceAlignedScreenCoordinates}.Offset and Scaled UVS
   B=Constant2Vector() {g=50, r=50}
  )
  n14 = StaticSwitchParameter(
   True=Multiply(A=VectorParameter() {parameterName=SpaceTint, defaultValue=<rgba 1 1 1 1>}.RGB, B=Multiply(A=@n11, B=@n12.RGB) {constB=1.4}) {constB=1.4}
   False=LinearInterpolate(
    B=TextureSample(UVs=@n13) {texture=&/T_SuperNovaTaro_Skin_Nebula}.RGB
    Alpha=ScalarParameter() {parameterName=Use Emissive Texture}
   ) {constA=1}
  ) {parameterName=Use Trail Texture}
  n15 = ParticleColor()
  MaterialFunctionCall(Emissive=Add(A=LinearInterpolate(B=@n2, Alpha=@n3), B=Multiply(A=@n14, B=@n15.RGB))) {materialFunction=&MF_FX_EmissiveAdjustments}
  n16 = ComponentMask(None=TextureCoordinate()) {g=true}
  n17 = Add(A=@n16, B=ScalarParameter() {group="Fade Top && Bottom Length Edges", parameterName=Top Length Fade Bias, sortPriority=2})
  n18 = ScalarParameter() {defaultValue=5, group="Fade Top && Bottom Length Edges", parameterName=Top Length Fade Scale}
  n19 = LinearInterpolate(
   B=Saturate(None=Multiply(A=@n17, B=@n18) {constB=3})
   Alpha=ScalarParameter() {defaultValue=1, group=Alpha, parameterName=Fade Top Length Edge, sortPriority=7}
  ) {constA=1}
  n20 = Add(A=OneMinus(None=@n16), B=ScalarParameter() {group="Fade Top && Bottom Length Edges", parameterName=Bottom Length Fade Bias, sortPriority=3})
  n21 = ScalarParameter() {defaultValue=5, group="Fade Top && Bottom Length Edges", parameterName=Bottom Length Fade Scale, sortPriority=1}
  n22 = LinearInterpolate(
   B=Saturate(None=Multiply(A=@n20, B=@n21) {constB=2})
   Alpha=ScalarParameter() {defaultValue=1, group=Alpha, parameterName=Fade Bottom Length Edge, sortPriority=8}
  ) {constA=1}
  n23 = VectorParameter() {group=Trailing Edge Sphere Mask, parameterName=Trailing Edge Sphereical UV Scale, sortPriority=4, defaultValue=<rgba 0.5 1 0 0>}
  n24 = VectorParameter() {group=Trailing Edge Sphere Mask, parameterName="Trailing Edge Sphere Mask Coord (B)", sortPriority=3, defaultValue=<rgba 0.85 0.82 0 0>}
  n25 = SphereMask(
   A=Multiply(A=TextureCoordinate(), B=ComponentMask(None=@n23.RGB) {g=true, r=true})
   B=ComponentMask(None=@n24.RGB) {g=true, r=true}
   Radius=ScalarParameter() {defaultValue=0.8, group=Trailing Edge Sphere Mask, parameterName=Trailing Edge Radius, sortPriority=1}
   Hardness=ScalarParameter() {defaultValue=0.3, group=Trailing Edge Sphere Mask, parameterName=Trailing Edge Hardness, sortPriority=2}
  )
  n26 = LinearInterpolate(
   B=Power(Base=OneMinus(None=@n25), Exp=ScalarParameter() {defaultValue=5, group=Trailing Edge Sphere Mask, parameterName=Trailing Edge Contrast}) {constExponent=5}
   Alpha=ScalarParameter() {defaultValue=1, group=Alpha, parameterName=Multiply Trailing Edge Sphere Mask, sortPriority=6}
  ) {constA=1}
  n27 = ComponentMask(None=TextureCoordinate()) {r=true}
  n28 = ScalarParameter() {group="Fade Leading && Trailing Edges", parameterName=Leading Edge Fade Bias, sortPriority=2}
  n29 = Saturate(None=Multiply(A=Add(A=@n27, B=@n28), B=ScalarParameter() {defaultValue=3, group="Fade Leading && Trailing Edges", parameterName=Leading Edge Fade Scale}) {constB=3})
  n30 = ScalarParameter() {defaultValue=1, group=Alpha, parameterName=Fade Leading Edge, sortPriority=4}
  n31 = OneMinus(None=@n27)
  n32 = ScalarParameter() {group="Fade Leading && Trailing Edges", parameterName=Trailing Edge Fade Bias, sortPriority=3}
  n33 = Saturate(None=Multiply(
    A=Add(A=@n31, B=@n32)
    B=ScalarParameter() {defaultValue=2, group="Fade Leading && Trailing Edges", parameterName=Trailing Edge Fade Scale, sortPriority=1}
   ) {constB=2})
  n34 = ScalarParameter() {defaultValue=1, group=Alpha, parameterName=Fade Trailing Edge, sortPriority=5}
  n35 = Multiply(
   A=Multiply(A=Multiply(A=@n19, B=@n22), B=Saturate(None=@n26))
   B=Multiply(A=LinearInterpolate(B=@n29, Alpha=@n30) {constA=1}, B=LinearInterpolate(B=@n33, Alpha=@n34) {constA=1})
  )
  n36 = MaterialFunctionCall("FadeSensitivity (optional)"=ScalarParameter() {defaultValue=1.2, parameterName="FadeSensitivity (optional) (S)"}) {materialFunction=&FadeBasedOnViewAngle}
  n37 = Multiply(B=Saturate(None=Multiply(A=@n36.Result) {constB=3})) {constA=1}
  n38 = TextureCoordinate()
  n39 = OneMinus(None=Subtract(A=ComponentMask(None=@n38) {g=true}) {constB=0.3})
  Multiply(
   A=Multiply(A=@n15.A, B=StaticSwitchParameter(True=@n35, False=@n37) {parameterName=Use Trail Texture})
   B=LinearInterpolate(B=Round(None=@n39), Alpha=ScalarParameter() {parameterName=Use bottom mask}) {constA=1}
  )
 ]
}
