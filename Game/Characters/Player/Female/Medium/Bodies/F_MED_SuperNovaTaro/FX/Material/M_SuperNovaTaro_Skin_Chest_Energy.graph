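# Translucent unlit material for an animated chest energy effect on the SuperNovaTaro character, enabled for Niagara mesh particles.
# Scrolling Perlin noise distorts panning plasma-coil and caustic textures, whose blend mixes two color parameters, tinted by particle color.
# Opacity combines particle alpha, a steep vertical UV gradient and three horizontal sine masks, two with offset parameters.
# Depth fade and view-angle fade soften the result.
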Material /Game/Characters/Player/Female/Medium/Bodies/F_MED_SuperNovaTaro/FX/Material/M_SuperNovaTaro_Skin_Chest_Energy
graph {
 properties {
  bUsedWithNiagaraMeshParticles=true
  blendMode=BLEND_Translucent
  pixelDepthOffsetMode=PDOM_Legacy
  shadingModel=MSM_Unlit
  usedShadingModels=Unlit
 }
 nodes [
  n1 = TextureCoordinate() {vTiling=3}
  n2 = TextureCoordinate() {vTiling=2}
  n3 = Multiply(A=TextureSample(UVs=Panner(Coordinate=@n2) {speedY=-0.3}) {samplerType=SAMPLERTYPE_LinearGrayscale, texture=&T_SimplePerlin_Greyscale}.R) {constB=0.2}
  n4 = TextureSample(UVs=Panner(Coordinate=Add(A=@n1, B=@n3)) {speedX=-0.5, speedY=-0.5}) {samplerType=SAMPLERTYPE_LinearGrayscale, texture=/Game/Effects/Fort_Effects/Emotes/UltraLight/Texture/T_PlasmaCoils}
  n5 = TextureCoordinate() {vTiling=3}
  n6 = TextureSample(UVs=Panner(Coordinate=Add(A=@n5, B=@n3)) {speedX=0.5, speedY=-0.5}) {texture=&T_ShieldCaustics}
  n7 = LinearInterpolate(
   A=VectorParameter() {parameterName=Color1, defaultValue=<rgba 0.0139 0 0.3 1>}.RGB
   B=VectorParameter() {parameterName=Color2, defaultValue=<rgba 2 0 0.0158 1>}.RGB
   Alpha=Multiply(A=LinearInterpolate(A=@n4.R, B=@n6.R))
  )
  n8 = ParticleColor()
  MaterialFunctionCall(Emissive=Multiply(A=Add(B=@n7), B=@n8.RGB)) {materialFunction=&MF_FX_EmissiveAdjustments}
  n9 = Saturate(None=Multiply(A=Multiply() {constA=1}) {constB=2})
  n10 = TextureCoordinate()
  n11 = Multiply(A=Power(Base=ComponentMask(None=@n10) {g=true}) {constExponent=5.5}) {constB=4000}
  n12 = Add(B=Multiply(A=Add(B=@n9), B=Saturate(None=@n11)))
  n13 = ComponentMask(None=TextureCoordinate()) {r=true}
  n14 = ScalarParameter() {defaultValue=-0.35, parameterName=offset mask}
  n15 = Subtract(A=Sine(None=Add(A=@n13, B=@n14) {constB=0}) {period=2}) {constB=0.93}
  n16 = Saturate(None=LinearInterpolate(Alpha=OneMinus(None=@n15)) {constA=-13})
  n17 = ComponentMask(None=TextureCoordinate()) {r=true}
  n18 = ScalarParameter() {defaultValue=0.2, parameterName=offset mask02}
  n19 = Subtract(A=Sine(None=Add(A=@n17, B=@n18) {constB=0}) {period=2}) {constB=0.93}
  n20 = Saturate(None=LinearInterpolate(Alpha=OneMinus(None=@n19)) {constA=-13})
  n21 = TextureCoordinate()
  n22 = Sine(None=Add(A=ComponentMask(None=@n21) {r=true}) {constB=0}) {period=2}
  n23 = LinearInterpolate(Alpha=OneMinus(None=Subtract(A=@n22) {constB=0.93})) {constA=-10}
  n24 = Constant() {r=0.5}
  n25 = Multiply(
   A=Multiply(A=Multiply(A=@n8.A, B=@n12), B=Multiply(A=@n16, B=@n20))
   B=Saturate(None=MaterialFunctionCall(In=@n23, Contrast=@n24) {materialFunction=&CheapContrast}.Result)
  )
  n26 = ScalarParameter() {defaultValue=1, parameterName="FadeSensitivity (optional) (S)"}
  Reroute(None=Multiply(
    A=DepthFade(Opacity=@n25) {fadeDistanceDefault=5}
    B=MaterialFunctionCall("FadeSensitivity (optional)"=@n26) {materialFunction=&FadeBasedOnViewAngle}.Result
   ))
 ]
}
