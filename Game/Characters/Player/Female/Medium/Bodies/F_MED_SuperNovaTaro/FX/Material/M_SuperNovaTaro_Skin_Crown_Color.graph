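# Translucent unlit crown color effect material for the SuperNovaTaro character, usable on Niagara mesh particles.
# Plasma coil texture pans in radial coordinates, distorted by slower-panning Perlin noise.
# The sharpened, boosted coils blend with a two-color radial gradient, are tinted by particle color and pass through emissive adjustments.
# Opacity combines particle alpha, a soft circular UV mask and depth fade.
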Material /Game/Characters/Player/Female/Medium/Bodies/F_MED_SuperNovaTaro/FX/Material/M_SuperNovaTaro_Skin_Crown_Color
graph {
 properties {
  bUsedWithNiagaraMeshParticles=true
  blendMode=BLEND_Translucent
  pixelDepthOffsetMode=PDOM_Legacy
  shadingModel=MSM_Unlit
  usedShadingModels=Unlit
 }
 nodes [
  n1 = Multiply(
   A=MaterialFunctionCall() {materialFunction=&VectorToRadialValue}.Radial Coordinates
   B=Constant2Vector() {g=1, r=1}
  )
  n2 = MaterialFunctionCall() {materialFunction=&VectorToRadialValue}
  n3 = Constant2Vector() {g=1, r=1}
  n4 = TextureSample(UVs=Panner(Coordinate=Multiply(A=@n2.Radial Coordinates, B=@n3)) {speedY=-0.1}) {samplerType=SAMPLERTYPE_LinearGrayscale, texture=&T_SimplePerlin_Greyscale}
  n5 = TextureSample(UVs=Add(A=Panner(Coordinate=@n1) {speedY=-0.4}, B=Multiply(A=@n4.R) {constB=0.2})) {samplerType=SAMPLERTYPE_LinearGrayscale, texture=/Game/Effects/Fort_Effects/Emotes/UltraLight/Texture/T_PlasmaCoils}
  n6 = SphereMask(A=TextureCoordinate(), B=Constant() {r=0.5}) {attenuationRadius=0.45, hardnessPercent=0}
  n7 = LinearInterpolate(
   A=VectorParameter() {parameterName=Color1, defaultValue=<rgba 2 0 0.3204 1>}.RGB
   B=VectorParameter() {parameterName=Color2, defaultValue=<rgba 0.0417 0 0.5 1>}.RGB
   Alpha=Saturate(None=Multiply(A=@n6) {constB=5})
  )
  n8 = Multiply(A=Multiply(A=Power(Base=@n5.R)) {constB=20}, B=@n7)
  n9 = ParticleColor()
  MaterialFunctionCall(Emissive=Multiply(A=LinearInterpolate(A=@n8, B=@n7), B=@n9.RGB)) {materialFunction=&MF_FX_EmissiveAdjustments}
  n10 = TextureCoordinate()
  n11 = Constant() {r=0.5}
  Multiply(
   A=@n9.A
   B=DepthFade(
    Opacity=SphereMask(A=@n10, B=@n11) {attenuationRadius=0.5, hardnessPercent=0}
    FadeDistance=ScalarParameter() {defaultValue=1, parameterName=FadeDistance}
   )
  )
 ]
}
