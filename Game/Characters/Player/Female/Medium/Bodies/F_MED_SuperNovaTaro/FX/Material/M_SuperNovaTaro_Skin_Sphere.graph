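# Unlit masked material for Niagara mesh particles, drawing a glowing nebula sphere for the SuperNovaTaro character effects.
# A panning, noise-distorted nebula texture is tinted and combined with a Fresnel inside-to-outside color gradient.
# Scalar blends optionally add radial-noise center glow and electric detail noise before particle color and emissive adjustments.
# Inverted Fresnel times particle alpha feeds DitherTemporalAA, softening the silhouette despite masked blending.
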Material /Game/Characters/Player/Female/Medium/Bodies/F_MED_SuperNovaTaro/FX/Material/M_SuperNovaTaro_Skin_Sphere
graph {
 properties {
  bUsedWithNiagaraMeshParticles=true
  blendMode=BLEND_Masked
  pixelDepthOffsetMode=PDOM_Legacy
  shadingModel=MSM_Unlit
  usedShadingModels=Unlit
 }
 nodes [
  n1 = Fresnel() {baseReflectFraction=0, exponent=1}
  n2 = Subtract(A=OneMinus(None=Power(Base=@n1) {constExponent=1})) {constB=0.6}
  n3 = MaterialFunctionCall() {materialFunction=&VectorToRadialValue}
  n4 = Constant2Vector() {g=1, r=5}
  n5 = TextureCoordinate()
  n6 = TextureSample!(
   UVs=Panner(Coordinate=Multiply(A=@n3.Radial Coordinates, B=@n4)) {speedY=-0.5}
   Tex
   "DDX(UVs)"=DDX(Value=@n5)
   "DDY(UVs)"=DDY(Value=@n5)
   Apply View MipBias
  ) {mipValueMode=TMVM_Derivative, samplerType=SAMPLERTYPE_Grayscale, texture=&LowResBlurredNoise}
  n7 = Multiply(A=@n2) {constB=2}
  n8 = Power(Base=@n6.R, Exp=LinearInterpolate(Alpha=Saturate(None=@n7)) {constA=5, constB=0})
  n9 = LinearInterpolate(
   B=Multiply(A=Constant3Vector() {constant=<rgba 10 2.4085 0.4167 1>}, B=Multiply(A=@n2, B=@n8))
   Alpha=ScalarParameter() {parameterName=Use Center Gradient}
  )
  n10 = Panner(Coordinate=TextureCoordinate() {uTiling=1.5}) {speedY=0.2}
  n11 = Power(Base=Multiply(
    A=Constant3Vector() {constant=<rgba 4 0.8149 0.375 1>}
    B=TextureSample!(UVs=@n10, Tex, Apply View MipBias) {samplerType=SAMPLERTYPE_LinearColor, texture=/Game/Characters/Player/Male/Medium/Bodies/M_MED_SRDrift/Textures/T_ElectricNoise}.R
   ))
  n12 = ScalarParameter() {parameterName=Use Detail Noise}
  n13 = TextureCoordinate() {uTiling=2, vTiling=2}
  n14 = TextureSample(UVs=Panner(Coordinate=TextureCoordinate()) {speedY=-0.1}) {samplerType=SAMPLERTYPE_LinearGrayscale, texture=/Game/Effects/Fort_Effects/Emotes/UltraLight/Texture/T_DirectionalNoise_Greyscale}
  n15 = TextureSampleParameter2D(UVs=Add(A=Panner(Coordinate=@n13) {speedX=0.1, speedY=0.2}, B=Multiply(A=@n14.R) {constB=0.2})) {parameterName=ColorTex, texture=&/T_SuperNovaTaro_Skin_Nebula}
  n16 = Multiply(A=VectorParameter() {parameterName=Color Tex, defaultValue=<rgba 1 1 1 1>}.RGB, B=@n15.RGB)
  n17 = ScalarParameter() {defaultValue=-2, parameterName=Des}
  n18 = Multiply(
   A=VectorParameter() {parameterName=Main Color, defaultValue=<rgba 2 0.6594 0.6817 1>}.RGB
   B=Power(Base=Desaturation(None=@n16, Fraction=@n17), Exp=ScalarParameter() {defaultValue=1.2, parameterName=ColorTexPower}) {constExponent=1.2}
  )
  n19 = VectorParameter() {parameterName=Color Inside, defaultValue=<rgba 0.0417 0 0.5 1>}
  n20 = ScalarParameter() {defaultValue=1, parameterName=ColorRange}
  n21 = Multiply(
   A=Power(Base=Fresnel(ExponentIn=@n20) {baseReflectFraction=0, exponent=1}, Exp=ScalarParameter() {defaultValue=3, parameterName=ColorRangPower}) {constExponent=3}
   B=ScalarParameter() {defaultValue=1, parameterName=ColorRangeEdge}
  )
  n22 = Multiply(A=LinearInterpolate(
    A=Add(A=@n19.RGB) {constB=0}
    B=VectorParameter() {parameterName=Color Outside, defaultValue=<rgba 2 0.0667 0 1>}.RGB
    Alpha=Saturate(None=@n21)
   ))
  n23 = Add(A=@n18, B=@n22)
  n24 = LinearInterpolate(A=@n18, B=@n22)
  n25 = ScalarParameter() {parameterName=Use Color Lerp}
  n26 = Multiply(
   A=Add(A=LinearInterpolate(B=@n11, Alpha=@n12), B=LinearInterpolate(A=@n23, B=@n24, Alpha=@n25))
   B=ScalarParameter() {defaultValue=1, parameterName=Emissive Intensity}
  )
  n27 = ParticleColor()
  MaterialFunctionCall(Emissive=Multiply(A=Add(A=@n9, B=@n26), B=@n27.RGB)) {materialFunction=&MF_FX_EmissiveAdjustments}
  n28 = Power(Base=OneMinus(None=Fresnel() {baseReflectFraction=0, exponent=4})) {constExponent=10}
  MaterialFunctionCall(Alpha Threshold=Multiply(A=Multiply(A=@n27.A, B=@n28))) {materialFunction=&DitherTemporalAA}
 ]
}
