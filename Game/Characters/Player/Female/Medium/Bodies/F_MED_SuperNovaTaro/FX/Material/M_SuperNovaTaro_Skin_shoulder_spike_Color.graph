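# Unlit translucent material for Niagara mesh particles, coloring the shoulder spike effect on the SuperNovaTaro skin with a nebula look.
# Blends a noise-distorted nebula texture with a noise-driven BumpOffset space texture for fake depth, tints it, then adds a vertical gradient with Fresnel rim color.
# Particle color tints emissive.
# Particle alpha, inverse-Fresnel edge fade, UV bottom mask and view-angle fade form the alpha feeding DepthFade and DitherTemporalAA.
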Material /Game/Characters/Player/Female/Medium/Bodies/F_MED_SuperNovaTaro/FX/Material/M_SuperNovaTaro_Skin_shoulder_spike_Color
graph {
 properties {
  bUsedWithNiagaraMeshParticles=true
  blendMode=BLEND_Translucent
  pixelDepthOffsetMode=PDOM_Legacy
  shadingModel=MSM_Unlit
  usedShadingModels=Unlit
 }
 nodes [
  n1 = ParticleColor()
  n2 = ScalarParameter() {defaultValue=1, parameterName=Use Fresnel fade}
  n3 = Power(Base=OneMinus(None=Fresnel() {baseReflectFraction=0, exponent=10})) {constExponent=10}
  n4 = OneMinus(None=ComponentMask(None=TextureCoordinate()) {g=true})
  n5 = ScalarParameter() {defaultValue=2, parameterName=Bottom mask power}
  n6 = Saturate(None=Multiply(A=Power(Base=@n4, Exp=@n5), B=ScalarParameter() {defaultValue=100, parameterName=Bottom mask edge}) {constB=100})
  n7 = ScalarParameter() {parameterName=AngleFade}
  n8 = Multiply(
   A=Multiply(A=@n1.A, B=LinearInterpolate(B=@n3, Alpha=@n2) {constA=1})
   B=Multiply(
    A=Multiply(A=@n6)
    B=MaterialFunctionCall("FadeSensitivity (optional)"=@n7) {materialFunction=&FadeBasedOnViewAngle}.Result
   )
  )
  DepthFade(Opacity=@n8) {fadeDistanceDefault=5}
  n9 = TextureCoordinate() {uTiling=2}
  n10 = Constant2Vector() {r=0.2}
  n11 = TextureSample(UVs=Panner(Coordinate=TextureCoordinate() {uTiling=3, vTiling=2}) {speedY=0.2}) {samplerType=SAMPLERTYPE_LinearGrayscale, texture=/Game/Effects/Fort_Effects/Emotes/UltraLight/Texture/T_DirectionalNoise_Greyscale}
  n12 = TextureSampleParameter2D(UVs=Add(A=Add(A=@n9, B=@n10), B=Multiply(A=@n11.R) {constB=0.1})) {parameterName=Color TexA, texture=&/T_SuperNovaTaro_Skin_Nebula}
  n13 = Constant()
  n14 = Panner(Coordinate=TextureCoordinate() {vTiling=2}) {speedY=0.1}
  n15 = BumpOffset(
   Coordinate=Add(A=TextureCoordinate() {uTiling=2}, B=Constant2Vector() {r=0.2})
   Height=Multiply(A=TextureSample(UVs=@n14) {samplerType=SAMPLERTYPE_LinearGrayscale, texture=&T_DirectionalNoise_Greyscale}.R) {constB=0.5}
  ) {heightRatio=-0.5}
  n16 = Multiply(
   A=VectorParameter() {parameterName=Color2, defaultValue=<rgba 2 0.4167 0.4167 1>}.RGB
   B=LinearInterpolate(
    A=Desaturation(None=@n12.RGB, Fraction=@n13)
    B=TextureSampleParameter2D(UVs=@n15) {parameterName=Color TexB, texture=/Game/Gadgets/Assets/VinderTech_GliderChute/Glider_Cube_Queen_Female/FX/T_Cheap_blue_space_color}.RGB
    Alpha=ScalarParameter() {defaultValue=0.8, parameterName=Lerp threshold}
   ) {constAlpha=0.8}
  )
  n17 = ComponentMask(None=TextureCoordinate()) {g=true}
  n18 = Multiply(A=Power(Base=OneMinus(None=@n17))) {constB=10}
  n19 = Power(Base=Fresnel() {baseReflectFraction=0, exponent=1}) {constExponent=3}
  n20 = LinearInterpolate(
   A=LinearInterpolate(
    A=VectorParameter() {parameterName=LerpColor1, defaultValue=<rgba 0.0333 0 0.4 1>}.RGB
    B=VectorParameter() {parameterName=LerpColor2, defaultValue=<rgba 0.0083 0 0.1 1>}.RGB
    Alpha=Saturate(None=@n18)
   )
   B=VectorParameter() {parameterName=Color1, defaultValue=<rgba 1 0.0333 0 1>}.RGB
   Alpha=Saturate(None=Multiply(A=@n19))
  )
  MaterialFunctionCall(Emissive=Multiply(A=Add(A=@n16, B=@n20), B=@n1.RGB)) {materialFunction=&MF_FX_EmissiveAdjustments}
  MaterialFunctionCall(Alpha Threshold=@n8) {materialFunction=&DitherTemporalAA}
 ]
}
