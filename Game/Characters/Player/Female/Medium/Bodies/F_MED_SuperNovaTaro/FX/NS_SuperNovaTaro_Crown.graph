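# Niagara system for the SuperNovaTaro character's crown, rendering a mesh particle attached to the body's skeletal mesh.
# A burst-spawned mesh particle with fixed orientation tweaks and camera offset follows sampled bones every update, its offset shifting for Big Head mode.
# Alpha scales with time of day, dimming at night.
# A user head-hidden bool drives its visibility tag, and distance culling sleeps and clears it.
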NiagaraSystem /Game/Characters/Player/Female/Medium/Bodies/F_MED_SuperNovaTaro/FX/NS_SuperNovaTaro_Crown
dependencies {usedDataInterfaces [], usedDynamicInputs [], usedModules [&SystemState]}
emitters {Crown_Color {
  emitterUpdateScript [
   &EmitterState(
    Enable Distance Culling=true
    Life Cycle Mode=::NewEnumerator1
    Loop Behavior=::NewEnumerator1
    Loop Duration Mode=::NewEnumerator1
    Max Distance Response=::SleepAndClearParticles
    MaxDistance=2000
    MinDistance=1950
    Module.bUseMaxDistance=true
    Module.bUseMinDistance=true
    Reset Age On Awaken=true
    Scalability Mode=::NewEnumerator1
    UseLoopDelay=false
   )
   &SpawnBurst_Instantaneous(Use Spawn Probability=false, UseLoopCountLimit=false)
  ]
  particleSpawnScript [
   &InitializeParticle(
    Color=<rgba 1 1 1 3>
    Mesh Scale=&Multiply_VectorByFloat(Vector=<1 1 1>, Float=0.98)
    Mesh Scale Mode=::NewEnumerator4
    UsePositionOffset=false
    Write Material Random=true
   )
   &InitialMeshOrientation(Mesh Orientation Mode=::NewEnumerator2, Rotation=<0 -0.17 0>, Use Rotation Vector=true)
   &InitialMeshOrientation(Mesh Orientation Mode=::NewEnumerator2, Rotation=<0 0 -0.04>, Use Rotation Vector=true) {module=InitialMeshOrientation001}
   &CameraOffset(Camera Offset Amount=0.5)
  ]
  particleUpdateScript [
   &ParticleState(Kill Particles When Lifetime Has Elapsed=false, Let Infinitely Lived Particles Die When Emitter Deactivates=true)
   &SkeletalMeshLocation(
    Bone Sampling Mode=::NewEnumerator4
    MaskSpawnGroup=false
    Orientation Sampling=::NewEnumerator1
    Sampled Position Offset=&Lerp_Vector(A=<0 -0.4 1.25>, B=<1.1824 -0.55 3>, Alpha=$NPC.FortniteNPC.BigHeadMode)
    Sampled Position Offset Coordinate Space:ENiagaraSkelMeshTransforms=::NewEnumerator4
    Skeletal Mesh=$User.Skeletal Mesh
   )
   &ScaleColor(
    Scale Alpha=&FNTimeOfDayScalar(Night Scale=0.5, Morning Scale=1, Evening Scale=1)
    ScaleA=true
    ScaleRGB=true
   )
   /SetVariables(Particles.VisibilityTag=&MakeCustomIntFromBool(Bool=$User.bisHeadHidden, True INT=1, False INT=0)) {isSetParameters=true}
  ]
  rendererClasses [&NiagaraMeshRendererProperties]
  renderers [{rendererClass=&NiagaraMeshRendererProperties, rendererIndex=0}]
 }}
summary {
 systemName=NS_SuperNovaTaro_Crown
 emitters [{emitterName=Crown_Color, rendererClasses [&NiagaraMeshRendererProperties]}]
 userVariables [
  User.bisHeadHidden {type=NiagaraBool}
  User.Skeletal Mesh {
   type=NiagaraDataInterfaceSkeletalMesh
   defaultValue {
    struct=NiagaraExt_VariableValue_DataInterface
    value {dataInterface=&self:NiagaraDataInterfaceSkeletalMesh_0}
   }
  }
 ]
}
