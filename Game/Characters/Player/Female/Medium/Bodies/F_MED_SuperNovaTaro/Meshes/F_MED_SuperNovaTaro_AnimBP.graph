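# Animation Blueprint for the SuperNovaTaro female medium player body mesh.
# It copies the pose from the attached parent mesh and runs it through a deform Control Rig.
# When ShortHair is true, additive bone-space offsets shift the attach_backpack and attach_cape bones.
# ShortHair is set on customization load from whether the character item carries the Stage1 cosmetic variant tag.
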AnimBlueprint /Game/Characters/Player/Female/Medium/Bodies/F_MED_SuperNovaTaro/Meshes/F_MED_SuperNovaTaro_AnimBP
target_skeleton=&/F_MED_SuperNovaTaro_Skeleton
graphs [
 AnimGraph {nodes [
   n1 = Control Rig(Source=Copy Pose From Mesh() {bUseAttachedParent=true}) {
    alphaBoolBlend {blendOption=Linear}
    controlRigAssetReference {blueprintRigClass=&/F_MED_SuperNovaTaro_DeformRig.F_MED_SuperNovaTaro_DeformRig_C}
   }
   n2 = Get ShortHair() -> (ShortHair:bool)
   n3 = "Transform (Modify) Bone - Bone: attach_backpack"!(ComponentPose=Local To Component(LocalPose=@n1), bAlphaBoolEnabled:bool=@n2, Translation="0,3,0", Rotation="0,0,0", Scale="1,1,1") -> (Pose) {
    alphaInputType=Bool
    translationMode=BMM_Additive
    translationSpace=BCS_BoneSpace
    alphaBoolBlend {blendOption=Linear}
    boneToModify {boneName=attach_backpack}
    translation {y=3}
   }
   n4 = "Transform (Modify) Bone - Bone: attach_cape"!(ComponentPose=@n3, bAlphaBoolEnabled:bool=@n2, Translation="0,3,-2", Rotation="0,0,0", Scale="1,1,1") -> (Pose) {
    alphaInputType=Bool
    translationMode=BMM_Additive
    translationSpace=BCS_BoneSpace
    alphaBoolBlend {blendOption=Linear}
    boneToModify {boneName=attach_cape}
    translation {y=3, z=-2}
   }
   Output Pose(Result=Component To Local(ComponentPose=@n4)) {layerGroup=None}
  ]}
 EventGraph {nodes [
   Event OnCharacterCustomizationLoaded()
   > n1 = Branch(Condition=CharacterItemDefinitionHasAnyMetaTag(Tags=Make Array("[0]"=(TagName="Cosmetics.Variant.Property.Stage1"))))
   > Set ShortHair(ShortHair:bool=true) -> (Output_Get:bool)
   Set ShortHair(execute=@n1.else, ShortHair=false)
  ]}
]
variables [ShortHair {type=boolean}]
