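# Material function in the SuperNovaTaro player body's materials folder that produces three-channel pseudo-random noise from a 2D input, defaulting to texture coordinates.
# It builds three float3 permutations of the input's two components and dots each with a different constant vector.
# The results form one vector that goes through the classic sine hash: sine, multiply by a large constant, then frac.
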MaterialFunction /Game/Characters/Player/Female/Medium/Bodies/F_MED_SuperNovaTaro/Materials/MF_RandomNoise
graph {nodes [
  n1 = FunctionInput(Preview=TextureCoordinate()) {bUsePreviewValueAsDefault=true, inputName=UVIndexInput, inputType=FunctionInput_Vector2}
  n2 = ComponentMask(None=@n1) {g=true}
  n3 = ComponentMask(None=@n1) {r=true}
  n4 = MaterialFunctionCall(X=@n2, Y=@n2, Z=@n3) {materialFunction=&MakeFloat3}
  n5 = Constant3Vector() {constant=<rgba 127.1 311.7 74.7 0>}
  n6 = MaterialFunctionCall(X=@n2, Y=@n3, Z=@n3) {materialFunction=&MakeFloat3}
  n7 = Constant3Vector() {constant=<rgba 269.5 183.3 246.1 0>}
  n8 = MaterialFunctionCall(X=@n3, Y=@n2, Z=@n2) {materialFunction=&MakeFloat3}
  n9 = Constant3Vector() {constant=<rgba 114 272 126 0>}
  n10 = Sine(None=MaterialFunctionCall(X=DotProduct(A=@n4.Result, B=@n5), Y=DotProduct(A=@n6.Result, B=@n7), Z=DotProduct(A=@n8.Result, B=@n9)) {materialFunction=&MakeFloat3}.Result)
  n11 = Constant() {r=14645}
  FunctionOutput(None=Frac(None=Multiply(A=@n10, B=@n11)))
 ]}
